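; Weapon Blueprint for a generic Playset Grenade consumable instance in Athena, judging by its name and path.
; It adds no graph logic of its own: BeginPlay, equip-and-attach and local unequip events are empty stubs, and the Tick event's parent call is a disabled placeholder.
; The construction script only calls its parent, so any behavior comes from the parent class or from settings outside these graphs.
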(event EventTick
  (params (DeltaSeconds :type Float%20%28single-precision%29))
  (|CallParentFunction DeltaSeconds) ; This node is disabled and will not be called.\nDrag off pins to build functionality.
)

(event EventBeginPlay)

(event FortWeapon|EventOnLocalUnEquip)

(event FortWeapon|EventOnEquippedAndAttached)

(fn ConstructionScript
  (|CallParentFunction))
