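; Gameplay ability instance for a playset grenade with a trajectory, supplying the playset to spawn.
; GetActivePlaysetData casts the ability's current source item definition to FortPlaysetGrenadeItemDefinition and resolves its soft PlaysetToSpawn reference, returning null if the cast fails.
; ShouldOnlyShowTrajectoryOnUse returns true, limiting the trajectory display to when the item is used.
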(fn GetActivePlaysetData
  (bind _asfort_playset_grenade_item_definition (Utilities|Casting|CastToFortPlaysetGrenadeItemDefinition (Ability|GetCurrentSourceItemDefinition))
    (:then
      (return (Utilities|ResolveSoftReference (Variables|/Script/FortniteGame.FortPlaysetGrenadeItemDefinition|GetPlaysetToSpawn _asfort_playset_grenade_item_definition))))
    (:CastFailed
      (return 0))))

(fn ShouldOnlyShowTrajectoryOnUse
  (return true))
